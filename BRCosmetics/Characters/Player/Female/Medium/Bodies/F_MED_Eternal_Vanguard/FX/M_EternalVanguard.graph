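# Translucent unlit two-sided FX material for the Eternal Vanguard character, drawing galaxy-filled, flame-like wisps on strands widened by SplineThicken.
# Noise-distorted, tapered sphere masks feed a curve atlas, shaping the colored edge glow and opacity.
# Panning screen-space nebula and star textures add interior emissive color.
# Opacity also fades by UV ends and sides, depth and Fresnel, while vertex color varies noise panning (R) and width (G).
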Material /BRCosmetics/Characters/Player/Female/Medium/Bodies/F_MED_Eternal_Vanguard/FX/M_EternalVanguard
graph {
 properties {
  bUsedWithMeshDeformer=true
  bUsedWithSkeletalMesh=true
  bUsedWithSplineMeshes=true
  blendMode=BLEND_Translucent
  pixelDepthOffsetMode=PDOM_Legacy
  shadingModel=MSM_Unlit
  translucencyPass=MTP_BeforeDOF
  twoSided=true
  usedShadingModels=Unlit
 }
 nodes [
  n1 = Multiply(A=ScreenPosition().ViewportUV, B=ScalarParameter() {defaultValue=4, parameterName=GalaxyTiling})
  n2 = TextureSample(
   UVs=Multiply(A=Panner(Coordinate=@n1) {speedY=-0.02}) {constB=2}
   Tex=TextureObject() {texture=/Game/Characters/Player/Male/Medium/Bodies/M_Med_Soldier_04/Skins/Galaxy/Textures/T_FN_Nebula}
  )
  n3 = Multiply(
   A=VectorParameter() {parameterName=BaseColor, defaultValue=<rgba 0.0232 1 0 1>}.RGB
   B=LinearInterpolate(Alpha=ComponentMask(None=@n2.RGB) {r=true}) {constA=0.25}
  )
  n4 = Constant() {r=-1.5}
  n5 = Multiply(
   A=Multiply(A=Desaturation(None=@n2.RGB, Fraction=@n4), B=ComponentMask(None=@n2.RGB) {b=true})
   B=VectorParameter() {parameterName=InsideColor, defaultValue=<rgba 1 0.7604 0.7604 1>}.RGB
  )
  n6 = ScreenPosition()
  n7 = ScalarParameter() {defaultValue=32, parameterName=Star_base_tile}
  n8 = TextureObject() {samplerType=SAMPLERTYPE_Masks, texture=/Game/Characters/Player/Male/Medium/Bodies/M_Med_Soldier_04/Skins/Galaxy/Textures/Celestial_T_Stars}
  n9 = TextureSample(UVs=Panner(Coordinate=Multiply(A=@n6.ViewportUV, B=@n7)) {speedY=0.02}, Tex=@n8)
  n10 = ScalarParameter() {defaultValue=0.5, parameterName=StarBrightness}
  n11 = Add(A=ComponentMask(None=TextureCoordinate()) {g=true}, B=ScalarParameter() {parameterName=Offset_color_diffuse}) {constB=0}
  n12 = Multiply(
   A=Add(A=Add(A=@n3, B=@n5), B=Multiply(A=@n9.B, B=@n10) {constB=2})
   B=LinearInterpolate(A=ScalarParameter() {defaultValue=0.1, parameterName=OffsetLow}, Alpha=Power(Base=@n11)) {constB=1.5}
  )
  n13 = Add(A=ComponentMask(None=TextureCoordinate()) {g=true}, B=ScalarParameter() {defaultValue=0.2, parameterName=Offset_color_Fresnel}) {constB=0}
  n14 = ScalarParameter() {defaultValue=2, parameterName=Taper1}
  n15 = ComponentMask(None=TextureCoordinate()) {r=true}
  n16 = ScalarParameter() {defaultValue=0.6, parameterName="Taper1_positon(UbaseV,0-1)"}
  n17 = ComponentMask(None=TextureCoordinate()) {g=true}
  n18 = Multiply(A=Divide(A=Constant() {r=1}, B=@n16), B=@n17)
  n19 = Constant() {r=1}
  n20 = Subtract(A=@n19, B=@n16)
  n21 = If(A=@n18, "A > B"=LinearInterpolate(A=Divide(A=@n19, B=@n20), Alpha=@n17) {constB=0}, "A == B"=@n18, "A < B"=@n18) {constB=1}
  n22 = ScalarParameter() {defaultValue=4, parameterName=Taper1_power}
  n23 = LinearInterpolate(
   A=LinearInterpolate(A=Multiply(A=@n14) {constB=-1}, B=Add(A=@n14), Alpha=@n15) {constA=-0.1, constB=1.1}
   B=@n15
   Alpha=OneMinus(None=Power(Base=@n21, Exp=@n22) {constExponent=4})
  )
  n24 = VertexColor()
  n25 = VectorParameter() {parameterName=sphere_bottom_texture, defaultValue=<rgba 0.1 0.9 1 1>}
  n26 = Multiply(
   A=AppendVector(A=LinearInterpolate(Alpha=@n24.R) {constA=-0.15, constB=0.15}, B=LinearInterpolate(Alpha=@n24.R) {constA=0.09, constB=0.2})
   B=@n25.B
  )
  n27 = Multiply(A=Time(), B=ScalarParameter() {defaultValue=0.9, parameterName=PanSpeed})
  n28 = ScalarParameter() {parameterName=MotionOffset}
  n29 = Multiply(A=Add(A=Panner(Time=@n27, Speed=@n26) {speedX=0.1, speedY=0.1}, B=@n28), B=ComponentMask(None=@n25.RGB) {g=true, r=true})
  n30 = ScalarParameter() {defaultValue=0.65, group=Flame_Area_Ctrl, parameterName=V_tile1}
  n31 = TextureCoordinate()
  n32 = Add(
   A=Multiply(
    A=TextureSample(UVs=@n29) {samplerType=SAMPLERTYPE_Normal, texture=/Game/Global/Textures/Default/Noise/T_SwirlNoise_N}.R
    B=ScalarParameter() {defaultValue=0.3, parameterName=Distort1}
   ) {constB=0.3}
   B=LinearInterpolate(B=@n30, Alpha=ComponentMask(None=@n31) {g=true})
  )
  n33 = Constant() {r=0.5}
  n34 = Add(A=@n30, B=ScalarParameter() {defaultValue=0.0452, parameterName=Sphere1_offset})
  n35 = VertexColor()
  n36 = Panner(
   Time=@n27
   Speed=AppendVector(A=LinearInterpolate(Alpha=@n35.R) {constA=-0.15, constB=0.15}, B=LinearInterpolate(Alpha=@n35.R) {constA=0.1, constB=0.2})
  ) {speedX=0.1, speedY=0.1}
  n37 = VectorParameter() {parameterName=sphere_top_texture, defaultValue=<rgba 0.2 0.5 1 1>}
  n38 = TextureSample(UVs=Multiply(A=Add(A=@n28, B=@n36), B=ComponentMask(None=@n37.RGB) {g=true, r=true})) {samplerType=SAMPLERTYPE_Normal, texture=&T_SwirlNoise_N}
  n39 = ScalarParameter() {defaultValue=0.3, parameterName=Distort2}
  n40 = ScalarParameter() {defaultValue=0.7704, group=Flame_Area_Ctrl, parameterName=V_tile2}
  n41 = ComponentMask(None=TextureCoordinate()) {g=true}
  n42 = AppendVector(
   A=ComponentMask(None=TextureCoordinate()) {r=true}
   B=Add(A=Multiply(A=@n38.R, B=@n39) {constB=0.3}, B=LinearInterpolate(B=@n40, Alpha=@n41))
  )
  n43 = AppendVector(A=Constant() {r=0.5}, B=Add(A=@n40, B=ScalarParameter() {defaultValue=-0.4786, parameterName=Sphere2_offset}) {constB=-0.8522})
  n44 = ScalarParameter() {defaultValue=0.1, group=Flame_Area_Ctrl, parameterName=Radius2, sortPriority=1}
  n45 = ScalarParameter() {group=Flame_Area_Ctrl, parameterName=Hardness2, sortPriority=2}
  n46 = Add(
   A=SphereMask(
    A=AppendVector(A=@n23, B=@n32)
    B=AppendVector(A=@n33, B=@n34)
    Radius=ScalarParameter() {defaultValue=0.6, group=Flame_Area_Ctrl, parameterName=Radius1, sortPriority=1}
    Hardness=ScalarParameter() {defaultValue=0.4, group=Flame_Area_Ctrl, parameterName=Hardness1, sortPriority=2}
   ) {attenuationRadius=1.5, hardnessPercent=0}
   B=StaticSwitchParameter(True=SphereMask(A=@n42, B=@n43, Radius=@n44, Hardness=@n45) {attenuationRadius=1.5, hardnessPercent=0}, False=Constant()) {defaultValue=true, parameterName=Use_secondSphere}
  ) {constB=0}
  n47 = CurveAtlasRowParameter(CurveTime=@n46) {
   atlas=&/CurveAtlas_EternalVanguard
   curve=&/Curve_EternalVanguard
   parameterName=Curve
   sortPriority=0
  }
  n48 = Multiply(
   A=LinearInterpolate(Alpha=Power(Base=@n13)) {constB=1.5}
   B=Multiply(A=VectorParameter() {parameterName=EdgeColor, defaultValue=<rgba 0.0422 0.3542 0.5 1>}.RGB, B=@n47.B)
  )
  MaterialFunctionCall(Emissive=Multiply(A=Add(A=@n48, B=@n12), B=ScalarParameter() {defaultValue=2, parameterName=Emissive_intensity})) {materialFunction=&MF_FX_EmissiveAdjustments}
  MaterialFunctionCall!(Position, UVRadialScale) {materialFunction=/Game/Weapons/WeaponSkins/MaterialFunction/ObjectStabilizedScreenSpaceUVs}
  n49 = TextureSample(UVs=Multiply(A=Panner(Time=@n27) {speedY=0.1}, B=Constant2Vector() {g=1, r=1})) {samplerType=SAMPLERTYPE_Normal, texture=&T_SwirlNoise_N}
  n50 = TextureCoordinate()
  n51 = CurveAtlasRowParameter(CurveTime=OneMinus(None=ComponentMask(None=@n50) {g=true})) {
   atlas=&CurveAtlas_EternalVanguard
   curve=&Curve_EternalVanguard
   parameterName=Curve
   sortPriority=0
  }
  n52 = CurveAtlasRowParameter(CurveTime=Multiply(A=LinearInterpolate(A=@n49.R, Alpha=@n51.G), B=@n46)) {
   atlas=&CurveAtlas_EternalVanguard
   curve=&Curve_EternalVanguard
   parameterName=Curve
   sortPriority=0
  }
  n53 = Multiply(A=@n51.G, B=@n52.R) {constA=1}
  n54 = OneMinus(None=ComponentMask(None=TextureCoordinate()) {g=true})
  n55 = ScalarParameter() {defaultValue=4, parameterName=FadePow}
  n56 = DepthFade(Opacity=@n53, FadeDistance=Multiply(A=Power(Base=@n54, Exp=@n55), B=ScalarParameter() {defaultValue=20, parameterName=FadeDistance}))
  n57 = Multiply(
   A=ShadingPathSwitch(Default=@n56, Deferred=@n56, Mobile=@n53) {desc=Migrated from FeatureLevelSwitch}
   B=ScalarParameter() {defaultValue=0.9, parameterName=OverallOpacity}
  )
  n58 = OneMinus(None=TextureCoordinate())
  n59 = Saturate(None=Multiply(A=ComponentMask(None=@n58) {g=true}, B=ScalarParameter() {defaultValue=3.5, parameterName=TipFadeMult}) {constB=6})
  n60 = Multiply(A=@n59, B=@n59)
  n61 = TextureCoordinate()
  n62 = Saturate(None=Multiply(A=ComponentMask(None=@n61) {g=true}, B=ScalarParameter() {defaultValue=6, parameterName=BtmFadeMult}) {constB=6})
  n63 = Multiply(A=@n62, B=@n62)
  n64 = Multiply(A=Multiply(A=@n60, B=@n60), B=Multiply(A=@n63, B=@n63))
  n65 = ComponentMask(None=TextureCoordinate()) {r=true}
  n66 = Multiply(A=Multiply(A=OneMinus(None=@n65), B=@n65)) {constB=4}
  n67 = LinearInterpolate(Alpha=VertexColor().G) {constA=1, constB=2}
  n68 = MaterialFunctionCall(
   WidthBase=Multiply(A=ScalarParameter() {defaultValue=6.5, parameterName=WidthBase}, B=@n67)
   WidthTip=Multiply(A=@n67, B=ScalarParameter() {defaultValue=10, parameterName=WidthTip})
  ) {materialFunction=&SplineThicken}
  n69 = OneMinus(None=Fresnel(Normal=@n68.Normal) {baseReflectFraction=0})
  n70 = ScalarParameter() {defaultValue=12, parameterName=FresnelPow}
  n71 = Constant() {r=1}
  Multiply(
   A=Multiply(A=Multiply(A=@n57, B=@n64), B=Multiply(A=@n66, B=@n66))
   B=QualitySwitch(Default=Power(Base=@n69, Exp=@n70), Low=@n71, Medium=@n71)
  )
  Reroute(None=@n68.WorldPositionOffset)
 ]
}
