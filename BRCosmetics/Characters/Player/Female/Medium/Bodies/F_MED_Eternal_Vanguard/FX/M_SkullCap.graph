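# Unlit translucent Niagara sprite and mesh material for the Eternal Vanguard skull cap FX, showing an animated galaxy look.
# A nebula texture and a star mask scroll in opposite directions in screen space, tinted and summed with an edge-color gradient derived from UV V as a fake Fresnel.
# Particle color and an intensity parameter scale the result through MF_FX_EmissiveAdjustments, while particle alpha times the skull cap mask's green channel forms opacity.
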Material /BRCosmetics/Characters/Player/Female/Medium/Bodies/F_MED_Eternal_Vanguard/FX/M_SkullCap
graph {
 properties {
  bUsedWithNiagaraMeshParticles=true
  bUsedWithNiagaraSprites=true
  blendMode=BLEND_Translucent
  pixelDepthOffsetMode=PDOM_Legacy
  shadingModel=MSM_Unlit
  translucencyPass=MTP_BeforeDOF
  usedShadingModels=Unlit
 }
 nodes [
  n1 = ParticleColor()
  n2 = Multiply(A=ScreenPosition().ViewportUV, B=ScalarParameter() {defaultValue=1, parameterName=GalaxyTiling})
  n3 = TextureSample(
   UVs=Multiply(A=Panner(Coordinate=@n2) {speedY=-0.02}) {constB=2}
   Tex=TextureObject() {texture=/Game/Characters/Player/Male/Medium/Bodies/M_Med_Soldier_04/Skins/Galaxy/Textures/T_FN_Nebula}
  )
  n4 = Multiply(
   A=VectorParameter() {parameterName=BaseColor, defaultValue=<rgba 0.1268 0.4013 0.7 1>}.RGB
   B=LinearInterpolate(Alpha=ComponentMask(None=@n3.RGB) {r=true}) {constA=0.25}
  )
  n5 = Constant() {r=-1.5}
  n6 = Multiply(
   A=Multiply(A=Desaturation(None=@n3.RGB, Fraction=@n5), B=ComponentMask(None=@n3.RGB) {b=true})
   B=VectorParameter() {parameterName=GalaxyColor, defaultValue=<rgba 0.1752 0.5921 1 1>}.RGB
  )
  n7 = ScreenPosition()
  n8 = ScalarParameter() {defaultValue=12, parameterName=Star_base_tile}
  n9 = TextureObject() {samplerType=SAMPLERTYPE_Masks, texture=/Game/Characters/Player/Male/Medium/Bodies/M_Med_Soldier_04/Skins/Galaxy/Textures/Celestial_T_Stars}
  n10 = TextureSample(UVs=Panner(Coordinate=Multiply(A=@n7.ViewportUV, B=@n8)) {speedY=0.02}, Tex=@n9)
  n11 = ScalarParameter() {defaultValue=2, parameterName=StarBrightness}
  n12 = TextureCoordinate()
  n13 = Power(Base=Add(A=ComponentMask(None=@n12) {g=true}, B=ScalarParameter() {defaultValue=0.2, parameterName=Offset_color_Fresnel}) {constB=0})
  n14 = VectorParameter() {parameterName=EdgeColor, defaultValue=<rgba 0.0422 0.3542 0.5 1>}
  n15 = Add(
   A=Multiply(A=LinearInterpolate(Alpha=@n13) {constB=1.5}, B=Multiply(A=@n14.RGB) {constB=0})
   B=Add(A=Add(A=@n4, B=@n6), B=Multiply(A=@n10.B, B=@n11) {constB=2})
  )
  MaterialFunctionCall(Emissive=Multiply(A=Multiply(A=@n1.RGB, B=@n15), B=ScalarParameter() {defaultValue=1, parameterName=Emissive_intensity})) {materialFunction=&MF_FX_EmissiveAdjustments}
  Multiply(
   A=ParticleColor().A
   B=TextureSampleParameter2D() {parameterName=OpacityTexture, texture=&/T_SkullCapMask}.G
  )
 ]
}
