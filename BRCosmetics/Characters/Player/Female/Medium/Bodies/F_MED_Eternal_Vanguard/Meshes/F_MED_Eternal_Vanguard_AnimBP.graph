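# Animation Blueprint for the female medium Eternal Vanguard cosmetic body, copying the pose from its attached parent mesh and adding rig-driven deformation.
# A per-bone layered blend pulls the neck_01 branch partly toward an unconnected pose input while restoring the copied pose from faceAttach.
# Three chained Control Rigs follow: base, twists and sculpters.
# Each update stores left upper-arm pitch relative to the clavicle in LeftArmUp, which the AnimGraph never reads.
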AnimBlueprint /BRCosmetics/Characters/Player/Female/Medium/Bodies/F_MED_Eternal_Vanguard/Meshes/F_MED_Eternal_Vanguard_AnimBP
target_skeleton=&/F_MED_Eternal_Vanguard_Skeleton
graphs [
 AnimGraph {nodes [
   n1 = Layered blend per bone!(
    BasePose=Use cached pose 'SavedPose29197'()
    BlendPoses_0
    BlendPoses_1:Pose Link Structure=Use cached pose 'SavedPose29197'()
    BlendWeights_0=0.7
    BlendWeights_1:float=1
   ) -> (Pose) {
    blendPoses [{linkId=-1, sourceLinkId=-1}, {linkId=-1, sourceLinkId=-1}]
    blendWeights [1,1]
    layerSetup [{branchFilters [{blendDepth=1, boneName=neck_01}]}, {branchFilters [{blendDepth=1, boneName=faceAttach}]}]
   }
   n2 = Control Rig(Source=@n1) {
    alphaBoolBlend {blendOption=Linear}
    controlRigAssetReference {blueprintRigClass=&/F_MED_Eternal_Vanguard_CtrlRig.F_MED_Eternal_Vanguard_CtrlRig_C}
   }
   n3 = Control Rig(Source=@n2) {
    alphaBoolBlend {blendOption=Linear}
    controlRigAssetReference {blueprintRigClass=&/F_MED_Eternal_Vanguard_Deform_CtrlRig_twists.F_MED_Eternal_Vanguard_Deform_CtrlRig_twists_C}
   }
   n4 = Control Rig!(
    Source=@n3
    Alpha
    KneeRot_fr_l_114deg:Vector
    KneeTrans_fr_l_114deg:Vector="0.258,-3.591,0.009"
    KneeScale_fr_l_114deg:Vector="1,1,1"
    KneeScale_fr_l_130deg:Vector="1,1,1"
    KneeRot_fr_l_130deg:Vector="0,0,-51.2"
    KneeTrans_fr_l_130deg:Vector="0.258,-3.591,0.009"
   ) -> (Pose) {
    alphaBoolBlend {blendOption=Linear}
    controlRigAssetReference {blueprintRigClass=&/F_MED_Eternal_Vanguard_Deform_CtrlRig_sculpters.F_MED_Eternal_Vanguard_Deform_CtrlRig_sculpters_C}
   }
   Output Pose(Result=@n4) {layerGroup=None}
   Save cached pose 'SavedPose29197'(Pose=Copy Pose From Mesh() {bUseAttachedParent=true}) {cacheName=SavedPose29197}
  ]}
 EventGraph {nodes [
   Event BlueprintUpdateAnimation()
   n1 = GetDeltaTransformFromRefPose(self=GetOwnerSkeletalMeshComponent(), BoneName=upperarm_l, BaseName=clavicle_l)
   Set LeftArmUp(LeftArmUp:double=BreakRotator(InRot=BreakTransform(InTransform=@n1).Rotation).Pitch) -> (Output_Get:double)
  ]}
]
variables [LeftArmUp {type=number}]
