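# Control Rig for the F_MED_Eternal_Vanguard female character that adds procedural motion to her hair joints.
# A spherical pose reader on the head, relative to spine_04, weights a rotation of the main hair root, followed by a fixed full-weight rotation offset.
# A per-item chain harmonics unit then animates that chain with wave and pendulum motion, using the head's initial position as reach target.
# Other harmonics units, including all left and right strand ones, are not connected to execution.
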ControlRigBlueprint /BRCosmetics/Characters/Player/Female/Medium/Bodies/F_MED_Eternal_Vanguard/Meshes/Parts/F_MED_Eternal_Vanguard_CtrlRig
graphs [
 RigVMFunctionLibrary {links [], nodes []}
 RigVMModel {nodes [
   /Script/ControlRig.RigUnit_ChainHarmonicsPerItem(
    ChainRoot=(Type=Bone,Name="dyn_hairsim_R1_jnt")
    DrawWorldOffset=(Rotation=(X=0,Y=0,Z=0,W=1),Translation=(X=0,Y=0,Z=0),Scale3D=(X=1,Y=1,Z=1))
    ExecutePin=()
    Pendulum=(bEnabled=True,PendulumStiffness=2,PendulumGravity=(X=0,Y=0,Z=0),PendulumBlend=0.75,PendulumDrag=0.98,PendulumMaximum=0.1,UnwindAxis=(X=0,Y=1,Z=0),UnwindMinimum=0.2,UnwindMaximum=0.05)
    Reach=(bEnabled=True,ReachTarget=(X=0,Y=0,Z=0),ReachAxis=(X=1,Y=0,Z=0))
    Speed=(X=1,Y=1,Z=1)
    Wave=(bEnabled=True,WaveFrequency=(X=0,Y=-3,Z=0),WaveAmplitude=(X=0,Y=0.5,Z=0),WaveOffset=(X=0,Y=-5,Z=0),WaveNoise=(X=0,Y=0,Z=0),WaveMaximum=1)
    WaveCurve=(EditorCurveData=(Keys=((),(Time=1,Value=1)),DefaultValue=340282346638528859811704183484516925440,PreInfinityExtrap=RCCE_Constant,PostInfinityExtrap=RCCE_Constant))
    bDrawDebug=True
   ) {is=struct RigVMUnitNode}
   > &RigUnit_ChainHarmonicsPerItem(
    ChainRoot=(Type=Bone,Name="dyn_hairsim_L1_jnt")
    DrawWorldOffset=(Rotation=(X=0,Y=0,Z=0,W=1),Translation=(X=0,Y=0,Z=0),Scale3D=(X=1,Y=1,Z=1))
    Pendulum=(bEnabled=True,PendulumStiffness=2,PendulumGravity=(X=0,Y=0,Z=0),PendulumBlend=0.75,PendulumDrag=0.98,PendulumMaximum=0.1,UnwindAxis=(X=0,Y=1,Z=0),UnwindMinimum=0.2,UnwindMaximum=0.05)
    Reach=(bEnabled=True,ReachTarget=(X=0,Y=0,Z=0),ReachAxis=(X=1,Y=0,Z=0))
    Speed=(X=1,Y=1,Z=1)
    Wave=(bEnabled=True,WaveFrequency=(X=0,Y=5,Z=0),WaveAmplitude=(X=0,Y=0.75,Z=0),WaveOffset=(X=0,Y=-5,Z=0),WaveNoise=(X=0,Y=0,Z=0),WaveMaximum=0.8)
    WaveCurve=(EditorCurveData=(Keys=((),(Time=1,Value=1)),DefaultValue=340282346638528859811704183484516925440,PreInfinityExtrap=RCCE_Constant,PostInfinityExtrap=RCCE_Constant))
    bDrawDebug=True
   )
   RigUnit_BeginExecution = &RigUnit_BeginExecution()
   n1 = &RigUnit_SphericalPoseReader(
    ActiveRegionScaleFactors=(PositiveWidth=0.7,NegativeWidth=0.2,PositiveHeight=0.5,NegativeHeight=0.5)
    ActiveRegionSize=0.45
    Debug=(bDrawDebug=false,DebugScale=25,DebugSegments=20,DebugThickness=0.25)
    DriverAxis=(X=1,Y=0,Z=0)
    DriverItem=(Type=Bone,Name="head")
    FalloffRegionScaleFactors=(PositiveWidth=0.5,NegativeWidth=0.5,PositiveHeight=0.5,NegativeHeight=0.5)
    FalloffSize=0.8
    OptionalParentItem=(Type=Bone,Name="spine_04")
    RotationOffset=(X=-90,Y=0,Z=-12)
    ExecutePin=@RigUnit_BeginExecution.ExecutePin
   )
   > &RigUnit_ModifyTransforms(
    ItemToModify=((Item=(Type=Bone,Name="dyn_hairsim_1_jnt"),Transform=(Rotation=(X=0,Y=0,Z=0,W=1),Translation=(X=0,Y=0,Z=0),Scale3D=(X=1,Y=1,Z=1))))
    ItemToModify.0.Transform.Rotation=&RigVMFunction_MathQuaternionFromEuler(Euler=(X=-15,Y=0,Z=0)).Result
    Weight=@n1.OutputParam
   )
   > &RigUnit_ModifyTransforms(
    ItemToModify=((Item=(Type=Bone,Name="dyn_hairsim_1_jnt"),Transform=(Rotation=(X=0,Y=0,Z=0,W=1),Translation=(X=0,Y=0,Z=0),Scale3D=(X=1,Y=1,Z=1))))
    Weight=1
    ItemToModify.0.Transform.Rotation=&RigVMFunction_MathQuaternionFromEuler(Euler=(X=0,Y=0,Z=-30)).Result
   )
   n2 = &RigUnit_GetTransform(Item=(Type=Bone,Name="head"), bInitial=true)
   > &RigUnit_ChainHarmonicsPerItem(
    ChainRoot=(Type=Bone,Name="dyn_hairsim_1_jnt")
    DrawWorldOffset=(Rotation=(X=0,Y=0,Z=0,W=1),Translation=(X=0,Y=0,Z=0),Scale3D=(X=1,Y=1,Z=1))
    Pendulum=(bEnabled=True,PendulumStiffness=2,PendulumGravity=(X=0,Y=0,Z=0),PendulumBlend=0.75,PendulumDrag=0.98,PendulumMaximum=0.1,UnwindAxis=(X=0,Y=1,Z=0),UnwindMinimum=0.2,UnwindMaximum=0.05)
    Reach=(bEnabled=True,ReachTarget=(X=0,Y=0,Z=0),ReachAxis=(X=1,Y=0,Z=0),ReachMinimum=0,ReachMaximum=0)
    Speed=(X=0.8,Y=0.8,Z=1)
    Wave=(bEnabled=True,WaveFrequency=(X=-1,Y=-1,Z=-4.5),WaveAmplitude=(X=0.1,Y=0.75,Z=-1),WaveOffset=(X=0,Y=0,Z=-5),WaveNoise=(X=0,Y=0,Z=0),WaveMinimum=-0.5,WaveMaximum=-1.25)
    WaveCurve=(EditorCurveData=(Keys=((Time=0.1619,Value=0.3394),(Time=0.5431,Value=0.407),(Time=0.8198,Value=0.6738))))
    bDrawDebug=True
    Reach.ReachTarget=@n2.Transform.Translation
   )
   &RigUnit_ChainHarmonicsPerItem(
    ChainRoot=(Type=Bone,Name="dyn_hairsim_1_jnt")
    DrawWorldOffset=(Rotation=(X=0,Y=0,Z=0,W=1),Translation=(X=0,Y=30,Z=0),Scale3D=(X=1,Y=1,Z=1))
    ExecutePin=()
    Pendulum=(bEnabled=True,PendulumStiffness=2,PendulumGravity=(X=0,Y=0,Z=0),PendulumBlend=0.75,PendulumDrag=0.98,PendulumMaximum=0.1,UnwindAxis=(X=0,Y=1,Z=0),UnwindMinimum=0.2,UnwindMaximum=0.05)
    Reach=(bEnabled=True,ReachTarget=(X=0,Y=0,Z=0),ReachAxis=(X=1,Y=0,Z=0),ReachMinimum=0,ReachMaximum=0)
    Speed=(X=0.75,Y=1,Z=0.75)
    Wave=(bEnabled=True,WaveFrequency=(X=-1,Y=-3,Z=-1),WaveAmplitude=(X=0.1,Y=0.75,Z=0.1),WaveOffset=(X=0,Y=0,Z=-5),WaveNoise=(X=0,Y=0,Z=0),WaveMinimum=0,WaveMaximum=1,WaveEase=QuadraticEaseInOut)
    WaveCurve=(EditorCurveData=(Keys=((Time=0.0681,Value=0.1586),(Time=0.3286,Value=0.6418),(Time=0.6575,Value=0.9126))))
    bDrawDebug=True
    Reach.ReachTarget=@n2.Transform.Translation
   )
   &RigUnit_ChainHarmonicsPerItem(
    ChainRoot=(Type=Bone,Name="dyn_hairsim_R1_jnt")
    DrawWorldOffset=(Rotation=(X=0,Y=0,Z=0,W=1),Translation=(X=0,Y=0,Z=0),Scale3D=(X=1,Y=1,Z=1))
    ExecutePin=()
    Pendulum=(bEnabled=True,PendulumStiffness=2,PendulumGravity=(X=0,Y=0,Z=0),PendulumBlend=0.75,PendulumDrag=0.98,PendulumMaximum=0.1,UnwindAxis=(X=0,Y=1,Z=0),UnwindMinimum=0.2,UnwindMaximum=0.05)
    Reach=(bEnabled=True,ReachTarget=(X=0,Y=0,Z=0),ReachAxis=(X=1,Y=0,Z=0))
    Speed=(X=0.75,Y=0.75,Z=0.75)
    Wave=(bEnabled=True,WaveFrequency=(X=-1,Y=-3,Z=-4),WaveAmplitude=(X=0.3,Y=0.3,Z=1),WaveOffset=(X=0,Y=0,Z=-5),WaveNoise=(X=0,Y=0,Z=0),WaveMaximum=1.25)
    WaveCurve=(EditorCurveData=(Keys=((Value=0.25),(Time=0.32,Value=0.45),(Time=0.68,Value=1))))
    bDrawDebug=True
   )
   &RigUnit_ChainHarmonicsPerItem(
    ChainRoot=(Type=Bone,Name="dyn_hairsim_1_jnt")
    DrawWorldOffset=(Rotation=(X=0,Y=0,Z=0,W=1),Translation=(X=0,Y=30,Z=0),Scale3D=(X=1,Y=1,Z=1))
    ExecutePin=()
    Pendulum=(bEnabled=True,PendulumStiffness=2,PendulumGravity=(X=0,Y=0,Z=0),PendulumBlend=0.75,PendulumDrag=0.98,PendulumMaximum=0.1,UnwindAxis=(X=0,Y=1,Z=0),UnwindMinimum=0.2,UnwindMaximum=0.05)
    Reach=(bEnabled=True,ReachTarget=(X=0,Y=0,Z=0),ReachAxis=(X=1,Y=0,Z=0),ReachMinimum=0,ReachMaximum=0)
    Speed=(X=0.75,Y=1,Z=1)
    Wave=(bEnabled=True,WaveFrequency=(X=-1,Y=-1,Z=0.3),WaveAmplitude=(X=0.1,Y=0.1,Z=1.25),WaveOffset=(X=0,Y=0,Z=-5),WaveNoise=(X=0,Y=0,Z=0),WaveMinimum=0,WaveMaximum=1.25,WaveEase=QuadraticEaseInOut)
    WaveCurve=(EditorCurveData=(Keys=((Time=0.0681,Value=0.1586),(Time=0.3455,Value=0.327),(Time=0.6575,Value=0.9126))))
    bDrawDebug=True
    Reach.ReachTarget=@n2.Transform.Translation
   )
  ]}
]
hierarchy [
 root {loc=<0 0 0>, rot=<0 0 0 1>}
 pelvis {parent=root, loc=<0 -0.5301 98.0571>, rot=<0.0127 0.707 -0.0127 -0.707>}
 spine_01 {parent=pelvis, loc=<6.1208 -0.9328 0>, rot=<0 0 0.034 -0.9994>}
 spine_02 {parent=spine_01, loc=<6.5996 -0.2095 0>, rot=<0 0 0.0238 -0.9997>}
 spine_03 {parent=spine_02, loc=<6.4914 -0.1237 0>, rot=<0 0 0.0149 -0.9999>}
 spine_04 {parent=spine_03, loc=<6.9598 0.0794 0>, rot=<0 0 -0.0132 -0.9999>}
 spine_05 {parent=spine_04, loc=<6.9987 0 0>, rot=<0 0 -0.0015 -1>}
 clavicle_l {parent=spine_05, loc=<5.4831 -0.1053 -3.2946>, rot=<0.0465 -0.7289 -0.009 -0.6829>}
 clavicle_r {parent=spine_05, loc=<5.4831 -0.1053 3.2946>, rot=<0.7289 0.0465 -0.6829 0.009>}
 neck_01 {parent=spine_05, loc=<11.3138 0.1529 0>, rot=<0 0 -0.2003 -0.9797>}
 neck_02 {parent=neck_01, loc=<4.5627 0 0>, rot=<0 0 0.0504 -0.9987>}
 head {parent=neck_02, loc=<5.3914 0 0>, rot=<0 0 0.115 -0.9934>}
 FX_hairsim_Look_at_jnt {parent=head, loc=<13.0125 0.6994 2.0345>, rot=<-0.0001 -0.9999 -0.0001 -0.0152>}
 FX_hairsim_MASTER_jnt {parent=FX_hairsim_Look_at_jnt, loc=<0 0 0>, rot=<0.0029 0.6963 0.003 -0.7178>}
 dyn_hairsim_L1_jnt {parent=FX_hairsim_MASTER_jnt, loc=<5.4318 2.8584 -1.7258>, rot=<0 -0.2589 0 -0.9659>}
 dyn_hairsim_L2_jnt {parent=dyn_hairsim_L1_jnt, loc=<5.2071 0 -0.0869>, rot=<0 -0.3828 0 -0.9238>}
 dyn_hairsim_L3_jnt {parent=dyn_hairsim_L2_jnt, loc=<12.0667 -2.8503 -1.5999>, rot=<0.0363 -0.2136 0.1636 -0.9624>}
 dyn_hairsim_L4_jnt {parent=dyn_hairsim_L3_jnt, loc=<7.6103 -0.9769 -0.7941>, rot=<0.01 -0.0232 0.0701 -0.9972>}
 dyn_hairsim_R1_jnt {parent=FX_hairsim_MASTER_jnt, loc=<-7.1674 -1.7307 -9.2898>, rot=<0 -0.6365 0 -0.7712>}
 dyn_hairsim_R2_jnt {parent=dyn_hairsim_R1_jnt, loc=<9.8281 -2.2347 1.5828>, rot=<0 0.3059 0 -0.9521>}
 dyn_hairsim_R3_jnt {parent=dyn_hairsim_R2_jnt, loc=<7.1902 -0.0423 0.366>, rot=<0 0.1627 0 -0.9867>}
 dyn_hairsim_1_jnt {parent=FX_hairsim_MASTER_jnt, loc=<2.851 -7.3605 -2.2544>, rot=<0.1013 -0.5817 0.2058 -0.7804>}
 dyn_hairsim_2_jnt {parent=dyn_hairsim_1_jnt, loc=<10.1724 2.278 0.7085>, rot=<0.0186 0.1242 -0.2756 -0.953>}
 dyn_hairsim_3_jnt {parent=dyn_hairsim_2_jnt, loc=<7.9472 0.7014 -1.5096>, rot=<0 0 -0.0035 -1>}
 dyn_hairsim_4_jnt {parent=dyn_hairsim_3_jnt, loc=<7.8888 2.7269 -2.0976>, rot=<-0.0275 -0.1795 -0.1524 -0.9715>}
]
