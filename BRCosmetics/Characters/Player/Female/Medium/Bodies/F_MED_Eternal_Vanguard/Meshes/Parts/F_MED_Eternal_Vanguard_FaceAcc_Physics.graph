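# Physics asset for the Eternal Vanguard female medium cosmetic's FaceAcc mesh part, set up for hair simulation.
# A four-link chain of dyn_hairsim capsules hangs from a tiny FX_hairsim_MASTER_jnt body on the head, with narrow swing and twist limits between links.
# Capsules on the head, neck, upper spine and clavicles form a body proxy, likely for hair collision, joined by free-rotation constraints.
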PhysicsAsset /BRCosmetics/Characters/Player/Female/Medium/Bodies/F_MED_Eternal_Vanguard/Meshes/Parts/F_MED_Eternal_Vanguard_FaceAcc_Physics
bodies {
 FX_hairsim_MASTER_jnt [{shapeType=Capsule, length=0.505, radius=0.505, shapeName=FX_hairsim_MASTER_jnt_capsule}]
 clavicle_l [{shapeType=Capsule, length=12.6717, radius=4.8754, shapeName=clavicle_l_capsule, center=<5 0 0>, rotation=<rot 90 0 0>}]
 clavicle_r [{shapeType=Capsule, length=13.1634, radius=4.8754, shapeName=clavicle_r_capsule, center=<-5 0 0>, rotation=<rot 90 0 0>}]
 dyn_hairsim_1_jnt [{shapeType=Capsule, length=12, radius=5, shapeName=dyn_hairsim_1_jnt_capsule, center=<4.7243 0 0>, rotation=<rot 0 0.0002 14.9999>}]
 dyn_hairsim_2_jnt [{shapeType=Capsule, length=12, radius=5, shapeName=dyn_hairsim_2_jnt_capsule, center=<3.2086 -1 0>, rotation=<rot 0 0 15>}]
 dyn_hairsim_3_jnt [{shapeType=Capsule, length=12, radius=5, shapeName=dyn_hairsim_3_jnt_capsule, center=<1 -3 0>, rotation=<rot 0 0.0001 19.9999>}]
 dyn_hairsim_4_jnt [{shapeType=Capsule, length=7.2991, radius=5.2154, shapeName=dyn_hairsim_4_jnt_capsule, center=<1 -12 0>}]
 head [{shapeType=Capsule, length=8.4367, radius=7.2805, shapeName=head_capsule, center=<-0.9804 2.5998 0>, rotation=<rot 90 90 72.5957>}]
 neck_01 [{shapeType=Capsule, length=4.255, radius=4.9, shapeName=neck_01_capsule, center=<-0.3384 0.7494 0>, rotation=<rot 0 -89.9999 0>}]
 spine_04 [{shapeType=Capsule, length=7.0044, radius=10.6649, shapeName=spine_04_capsule, center=<2.5931 1.6603 0>, rotation=<rot 0 -92.6845 0>}]
 spine_05 [{shapeType=Capsule, length=17.0459, radius=8.0403, shapeName=spine_05_capsule, center=<1.675 -0.1132 -0.155>}]
}
constraints [
 {
  bone1Name=spine_05
  bone2Name=spine_04
  swing1LimitDegrees=45
  swing1Motion=Free
  swing2LimitDegrees=45
  swing2Motion=Free
  twistLimitDegrees=45
  twistMotion=Free
 }
 {
  bone1Name=clavicle_l
  bone2Name=spine_05
  swing1LimitDegrees=45
  swing1Motion=Free
  swing2LimitDegrees=45
  swing2Motion=Free
  twistLimitDegrees=45
  twistMotion=Free
 }
 {
  bone1Name=clavicle_r
  bone2Name=spine_05
  swing1LimitDegrees=45
  swing1Motion=Free
  swing2LimitDegrees=45
  swing2Motion=Free
  twistLimitDegrees=45
  twistMotion=Free
 }
 {
  bone1Name=head
  bone2Name=neck_01
  swing1LimitDegrees=45
  swing1Motion=Free
  swing2LimitDegrees=45
  swing2Motion=Free
  twistLimitDegrees=45
  twistMotion=Free
 }
 {
  bone1Name=neck_01
  bone2Name=spine_05
  swing1LimitDegrees=45
  swing1Motion=Free
  swing2LimitDegrees=45
  swing2Motion=Free
  twistLimitDegrees=45
  twistMotion=Free
 }
 {bone1Name=dyn_hairsim_1_jnt, bone2Name=FX_hairsim_MASTER_jnt, swing1LimitDegrees=25, swing2LimitDegrees=15, twistLimitDegrees=3}
 {bone1Name=dyn_hairsim_2_jnt, bone2Name=dyn_hairsim_1_jnt, swing1LimitDegrees=25, swing2LimitDegrees=15, twistLimitDegrees=3}
 {bone1Name=dyn_hairsim_3_jnt, bone2Name=dyn_hairsim_2_jnt, swing1LimitDegrees=25, swing2LimitDegrees=15, twistLimitDegrees=3}
 {bone1Name=dyn_hairsim_4_jnt, bone2Name=dyn_hairsim_3_jnt, swing1LimitDegrees=25, swing2LimitDegrees=15, twistLimitDegrees=3}
 {
  bone1Name=FX_hairsim_MASTER_jnt
  bone2Name=head
  swing1LimitDegrees=45
  swing1Motion=Free
  swing2LimitDegrees=45
  swing2Motion=Free
  twistLimitDegrees=45
  twistMotion=Free
 }
]
