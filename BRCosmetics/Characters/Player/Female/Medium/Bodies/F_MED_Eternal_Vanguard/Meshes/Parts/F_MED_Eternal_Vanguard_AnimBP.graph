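# Post-attach AnimBP for the Eternal Vanguard female cosmetic's mesh parts, adding secondary motion to a pose copied from the parent mesh.
# The output chain runs Copy Pose From Mesh through RigidBody (FaceAcc physics asset, spine_05-relative, reduced gravity) into the character's Control Rig.
# AnimDynamics, bone transform, constraint and Look At hair chains are unconnected to the output.
# EventGraph still computes four phase-shifted sine rotators each update, read only by those unconnected transforms.
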AnimBlueprint /BRCosmetics/Characters/Player/Female/Medium/Bodies/F_MED_Eternal_Vanguard/Meshes/Parts/F_MED_Eternal_Vanguard_AnimBP
target_skeleton=&/F_MED_Eternal_Vanguard_Skeleton
graphs [
 AnimGraph {nodes [
   n1 = Anim Dynamics - Bone: dyn_hairsim_L2_jnt(Alpha=1, ExternalForce="0,0,0") {
    angularDampingOverride=1
    bOverrideAngularDamping=true
    bUseGravityOverride=true
    simulationSpace=RootRelative
    alphaBoolBlend {blendOption=Linear}
    boundBone {boneName=dyn_hairsim_L2_jnt}
    componentLinearAccScale=<0.4 0.4 0.4>
    componentLinearVelScale=<0.4 0.4 0.4>
    gravityOverride {z=-980}
    physicsBodyDefinitions [{
      boundBone {boneName=dyn_hairsim_L2_jnt}
      boxExtents=<40 40 40>
      constraintSetup {angularLimitsMax=<20 20 20>, angularLimitsMin=<-80 -80 -80>, angularTarget=<1 0 0>}
      localJointOffset=<20 0 0>
     }]
   }
   n2 = Anim Dynamics - Bone: dyn_hairsim_L3_jnt(ComponentPose=@n1, Alpha=1, ExternalForce="0,0,0") {
    angularDampingOverride=1
    bOverrideAngularDamping=true
    bUseGravityOverride=true
    simulationSpace=RootRelative
    alphaBoolBlend {blendOption=Linear}
    boundBone {boneName=dyn_hairsim_L3_jnt}
    componentLinearAccScale=<0.4 0.4 0.4>
    componentLinearVelScale=<0.4 0.4 0.4>
    gravityOverride {z=-980}
    physicsBodyDefinitions [{
      boundBone {boneName=dyn_hairsim_L3_jnt}
      boxExtents=<35 35 35>
      constraintSetup {angularLimitsMax=<20 20 20>, angularLimitsMin=<-80 -80 -80>, angularTarget=<1 0 0>}
      localJointOffset=<17 0 0>
     }]
   }
   n3 = Anim Dynamics - Bone: dyn_hairsim_L4_jnt(ComponentPose=@n2, Alpha=1, ExternalForce="0,0,0") {
    angularDampingOverride=1
    bOverrideAngularDamping=true
    bUseGravityOverride=true
    simulationSpace=RootRelative
    alphaBoolBlend {blendOption=Linear}
    boundBone {boneName=dyn_hairsim_L4_jnt}
    componentLinearAccScale=<0.4 0.4 0.4>
    componentLinearVelScale=<0.4 0.4 0.4>
    gravityOverride {z=-980}
    physicsBodyDefinitions [{
      boundBone {boneName=dyn_hairsim_L4_jnt}
      boxExtents=<50 50 50>
      constraintSetup {angularLimitsMax=<20 20 20>, angularLimitsMin=<-80 -80 -80>, angularTarget=<1 0 0>}
      localJointOffset=<25 0 0>
     }]
   }
   n4 = Anim Dynamics - Bone: dyn_hairsim_1_jnt(ComponentPose=@n3, Alpha=1, ExternalForce="0,0,0") {
    angularDampingOverride=1
    bOverrideAngularDamping=true
    bUseGravityOverride=true
    simulationSpace=RootRelative
    alphaBoolBlend {blendOption=Linear}
    boundBone {boneName=dyn_hairsim_1_jnt}
    componentLinearAccScale=<0.4 0.4 0.4>
    componentLinearVelScale=<0.4 0.4 0.4>
    gravityOverride {z=-980}
    physicsBodyDefinitions [{
      boundBone {boneName=dyn_hairsim_1_jnt}
      boxExtents=<25 25 25>
      constraintSetup {angularLimitsMax=<5 5 5>, angularLimitsMin=<-60 -60 -60>, angularTarget=<1 0 0>}
      localJointOffset=<12 0 0>
     }]
   }
   n5 = Anim Dynamics - Bone: dyn_hairsim_2_jnt(ComponentPose=@n4, Alpha=1, ExternalForce="0,0,0") {
    angularDampingOverride=1
    bOverrideAngularDamping=true
    bUseGravityOverride=true
    simulationSpace=RootRelative
    alphaBoolBlend {blendOption=Linear}
    boundBone {boneName=dyn_hairsim_2_jnt}
    componentLinearAccScale=<0.4 0.4 0.4>
    componentLinearVelScale=<0.4 0.4 0.4>
    gravityOverride {z=-980}
    physicsBodyDefinitions [{
      boundBone {boneName=dyn_hairsim_2_jnt}
      boxExtents=<25 25 25>
      constraintSetup {angularLimitsMax=<5 5 5>, angularLimitsMin=<-60 -60 -60>, angularTarget=<1 0 0>}
      localJointOffset=<12 0 0>
     }]
   }
   n6 = Anim Dynamics - Bone: dyn_hairsim_3_jnt(ComponentPose=@n5, Alpha=1, ExternalForce="0,0,0") {
    angularDampingOverride=1
    bOverrideAngularDamping=true
    bUseGravityOverride=true
    simulationSpace=RootRelative
    alphaBoolBlend {blendOption=Linear}
    boundBone {boneName=dyn_hairsim_3_jnt}
    componentLinearAccScale=<0.4 0.4 0.4>
    componentLinearVelScale=<0.4 0.4 0.4>
    gravityOverride {z=-980}
    physicsBodyDefinitions [{
      boundBone {boneName=dyn_hairsim_3_jnt}
      boxExtents=<35 35 35>
      constraintSetup {angularLimitsMax=<5 5 5>, angularLimitsMin=<-60 -60 -60>, angularTarget=<1 0 0>}
      localJointOffset=<17 0 0>
     }]
   }
   Anim Dynamics - Bone: dyn_hairsim_4_jnt(ComponentPose=@n6, Alpha=1, ExternalForce="0,0,0") {
    angularDampingOverride=1
    bOverrideAngularDamping=true
    bUseGravityOverride=true
    simulationSpace=RootRelative
    alphaBoolBlend {blendOption=Linear}
    boundBone {boneName=dyn_hairsim_4_jnt}
    componentLinearAccScale=<0.4 0.4 0.4>
    componentLinearVelScale=<0.4 0.4 0.4>
    gravityOverride {z=-980}
    physicsBodyDefinitions [{
      boundBone {boneName=dyn_hairsim_4_jnt}
      boxExtents=<50 50 50>
      constraintSetup {angularLimitsMax=<5 5 5>, angularLimitsMin=<-60 -60 -60>, angularTarget=<1 0 0>}
      localJointOffset=<25 0 0>
     }]
   }
   n7 = Get sineRotation1() -> (sineRotation1:Rotator)
   n8 = "Transform (Modify) Bone - Bone: dyn_hairsim_1_jnt"(Alpha=1, Translation="0,0,0", Rotation=@n7, Scale="1,1,1") {
    rotationMode=BMM_Additive
    alphaBoolBlend {blendOption=Linear}
    alphaScaleBiasClamp {scale=0.5}
    boneToModify {boneName=dyn_hairsim_1_jnt}
    rotation=<rot 65.899 36.555 103.446>
   }
   n9 = Get sineRotation2() -> (sineRotation2:Rotator)
   n10 = "Transform (Modify) Bone - Bone: dyn_hairsim_2_jnt"(ComponentPose=@n8, Alpha=1, Translation="0,0,0", Rotation=@n9, Scale="1,1,1") {rotationMode=BMM_Additive, alphaBoolBlend {blendOption=Linear}, alphaScaleBiasClamp {scale=0.5}, boneToModify {boneName=dyn_hairsim_2_jnt}}
   n11 = Get sineRotation3() -> (sineRotation3:Rotator)
   n12 = "Transform (Modify) Bone - Bone: dyn_hairsim_3_jnt"(ComponentPose=@n10, Alpha=1, Translation="0,0,0", Rotation=@n11, Scale="1,1,1") {rotationMode=BMM_Additive, alphaBoolBlend {blendOption=Linear}, alphaScaleBiasClamp {scale=0.5}, boneToModify {boneName=dyn_hairsim_3_jnt}}
   n13 = Get sineRotation4() -> (sineRotation4:Rotator)
   n14 = "Transform (Modify) Bone - Bone: dyn_hairsim_4_jnt"(ComponentPose=@n12, Alpha=1, Translation="0,0,0", Rotation=@n13, Scale="1,1,1") {rotationMode=BMM_Additive, alphaBoolBlend {blendOption=Linear}, alphaScaleBiasClamp {scale=0.5}, boneToModify {boneName=dyn_hairsim_4_jnt}}
   n15 = Anim Dynamics - Bone: dyn_hairsim_1_jnt(ComponentPose=@n14, Alpha=1, ExternalForce="0,0,0") {
    angularDampingOverride=0.9
    angularSpringConstant=500
    bAngularSpring=true
    bOverrideAngularDamping=true
    bUseGravityOverride=true
    gravityScale=0.35
    simulationSpace=Actor
    alphaBoolBlend {blendOption=Linear}
    boundBone {boneName=dyn_hairsim_1_jnt}
    componentLinearAccScale=<1.2 1.2 1.1>
    componentLinearVelScale=<1.2 1.2 1.1>
    gravityOverride {z=-540}
    physicsBodyDefinitions [{
      sphereCollisionRadius=0
      boundBone {boneName=dyn_hairsim_1_jnt}
      boxExtents=<50 60 60>
      constraintSetup {angularLimitsMax=<15 15 0>, angularLimitsMin=<-30 -15 -30>, angularTarget=<1 0 0>}
      localJointOffset=<4 0 0>
     }]
    relativeSpaceBone {boneName=pelvis}
   }
   n16 = Anim Dynamics - Bone: dyn_hairsim_2_jnt(ComponentPose=@n15, Alpha=1, ExternalForce="0,0,0") {
    angularDampingOverride=0.9
    angularSpringConstant=500
    bAngularSpring=true
    bOverrideAngularDamping=true
    bUseGravityOverride=true
    gravityScale=0.35
    simulationSpace=Actor
    alphaBoolBlend {blendOption=Linear}
    boundBone {boneName=dyn_hairsim_2_jnt}
    componentLinearAccScale=<1.2 1.2 1.1>
    componentLinearVelScale=<1.2 1.2 1.1>
    gravityOverride {z=-540}
    physicsBodyDefinitions [{
      sphereCollisionRadius=0
      boundBone {boneName=dyn_hairsim_2_jnt}
      boxExtents=<50 60 60>
      constraintSetup {angularLimitsMax=<15 15 30>, angularLimitsMin=<-30 -15 -30>, angularTarget=<1 0 0>}
      localJointOffset=<4 0 0>
     }]
    relativeSpaceBone {boneName=pelvis}
   }
   n17 = Anim Dynamics - Bone: dyn_hairsim_3_jnt(ComponentPose=@n16, Alpha=1, ExternalForce="0,0,0") {
    angularDampingOverride=0.9
    angularSpringConstant=500
    bAngularSpring=true
    bOverrideAngularDamping=true
    bUseGravityOverride=true
    gravityScale=0.35
    simulationSpace=Actor
    alphaBoolBlend {blendOption=Linear}
    boundBone {boneName=dyn_hairsim_3_jnt}
    componentLinearAccScale=<1.2 1.2 1.1>
    componentLinearVelScale=<1.2 1.2 1.1>
    gravityOverride {z=-540}
    physicsBodyDefinitions [{
      sphereCollisionRadius=0
      boundBone {boneName=dyn_hairsim_3_jnt}
      boxExtents=<50 60 60>
      constraintSetup {angularLimitsMax=<15 15 30>, angularLimitsMin=<-30 -15 -30>, angularTarget=<1 0 0>}
      localJointOffset=<4 0 0>
     }]
    relativeSpaceBone {boneName=pelvis}
   }
   Anim Dynamics - Bone: dyn_hairsim_4_jnt(ComponentPose=@n17, Alpha=1, ExternalForce="0,0,0") {
    angularDampingOverride=0.9
    angularSpringConstant=500
    bAngularSpring=true
    bOverrideAngularDamping=true
    bUseGravityOverride=true
    gravityScale=0.35
    simulationSpace=Actor
    alphaBoolBlend {blendOption=Linear}
    boundBone {boneName=dyn_hairsim_4_jnt}
    componentLinearAccScale=<1.2 1.2 1.1>
    componentLinearVelScale=<1.2 1.2 1.1>
    gravityOverride {z=-540}
    physicsBodyDefinitions [{
      sphereCollisionRadius=0
      boundBone {boneName=dyn_hairsim_4_jnt}
      boxExtents=<50 60 60>
      constraintSetup {angularLimitsMax=<15 15 30>, angularLimitsMin=<-30 -15 -30>, angularTarget=<1 0 0>}
      localJointOffset=<4 0 0>
     }]
    relativeSpaceBone {boneName=pelvis}
   }
   n18 = Anim Dynamics - Bone: dyn_hairsim_L1_jnt(Alpha=1, ExternalForce="0,0,0") {
    angularDampingOverride=0.9
    angularSpringConstant=500
    bAngularSpring=true
    bOverrideAngularDamping=true
    bUseGravityOverride=true
    gravityScale=0.35
    simulationSpace=Actor
    alphaBoolBlend {blendOption=Linear}
    boundBone {boneName=dyn_hairsim_L1_jnt}
    componentLinearAccScale=<1.2 1.2 1.1>
    componentLinearVelScale=<1.2 1.2 1.1>
    gravityOverride {z=-720}
    physicsBodyDefinitions [{
      sphereCollisionRadius=0
      boundBone {boneName=dyn_hairsim_L1_jnt}
      boxExtents=<50 60 60>
      constraintSetup {angularLimitsMax=<15 0 30>, angularLimitsMin=<-15 -30 -15>, angularTarget=<1 0 0>}
      localJointOffset=<2 0 0>
     }]
    relativeSpaceBone {boneName=pelvis}
   }
   n19 = Anim Dynamics - Bone: dyn_hairsim_L2_jnt(ComponentPose=@n18, Alpha=1, ExternalForce="0,0,0") {
    angularDampingOverride=0.9
    angularSpringConstant=500
    bAngularSpring=true
    bOverrideAngularDamping=true
    bUseGravityOverride=true
    gravityScale=0.35
    simulationSpace=Actor
    alphaBoolBlend {blendOption=Linear}
    boundBone {boneName=dyn_hairsim_L2_jnt}
    componentLinearAccScale=<1.2 1.2 1.1>
    componentLinearVelScale=<1.2 1.2 1.1>
    gravityOverride {z=-720}
    physicsBodyDefinitions [{
      sphereCollisionRadius=0
      boundBone {boneName=dyn_hairsim_L2_jnt}
      boxExtents=<50 60 60>
      constraintSetup {angularLimitsMax=<15 0 30>, angularLimitsMin=<-15 -30 -15>, angularTarget=<1 0 0>}
      localJointOffset=<2 0 0>
     }]
    relativeSpaceBone {boneName=pelvis}
   }
   n20 = Anim Dynamics - Bone: dyn_hairsim_L3_jnt(ComponentPose=@n19, Alpha=1, ExternalForce="0,0,0") {
    angularDampingOverride=0.9
    angularSpringConstant=500
    bAngularSpring=true
    bOverrideAngularDamping=true
    bUseGravityOverride=true
    gravityScale=0.35
    simulationSpace=Actor
    alphaBoolBlend {blendOption=Linear}
    boundBone {boneName=dyn_hairsim_L3_jnt}
    componentLinearAccScale=<1.2 1.2 1.1>
    componentLinearVelScale=<1.2 1.2 1.1>
    gravityOverride {z=-720}
    physicsBodyDefinitions [{
      sphereCollisionRadius=0
      boundBone {boneName=dyn_hairsim_L3_jnt}
      boxExtents=<50 60 60>
      constraintSetup {angularLimitsMax=<15 0 30>, angularLimitsMin=<-15 -30 -15>, angularTarget=<1 0 0>}
      localJointOffset=<2 0 0>
     }]
    relativeSpaceBone {boneName=pelvis}
   }
   n21 = Anim Dynamics - Bone: dyn_hairsim_L4_jnt(ComponentPose=@n20, Alpha=1, ExternalForce="0,0,0") {
    angularDampingOverride=0.9
    angularSpringConstant=500
    bAngularSpring=true
    bOverrideAngularDamping=true
    bUseGravityOverride=true
    gravityScale=0.35
    simulationSpace=Actor
    alphaBoolBlend {blendOption=Linear}
    boundBone {boneName=dyn_hairsim_L4_jnt}
    componentLinearAccScale=<1.2 1.2 1.1>
    componentLinearVelScale=<1.2 1.2 1.1>
    gravityOverride {z=-720}
    physicsBodyDefinitions [{
      sphereCollisionRadius=0
      boundBone {boneName=dyn_hairsim_L4_jnt}
      boxExtents=<50 60 60>
      constraintSetup {angularLimitsMax=<15 0 30>, angularLimitsMin=<-15 -30 -15>, angularTarget=<1 0 0>}
      localJointOffset=<2 0 0>
     }]
    relativeSpaceBone {boneName=pelvis}
   }
   n22 = Anim Dynamics - Bone: dyn_hairsim_R1_jnt(ComponentPose=@n21, Alpha=1, ExternalForce="0,0,0") {
    angularDampingOverride=0.9
    angularSpringConstant=500
    bAngularSpring=true
    bOverrideAngularDamping=true
    bUseGravityOverride=true
    gravityScale=0.35
    simulationSpace=Actor
    alphaBoolBlend {blendOption=Linear}
    boundBone {boneName=dyn_hairsim_R1_jnt}
    componentLinearAccScale=<1.2 1.2 1.1>
    componentLinearVelScale=<1.2 1.2 1.1>
    gravityOverride {z=-720}
    physicsBodyDefinitions [{
      sphereCollisionRadius=0
      boundBone {boneName=dyn_hairsim_R1_jnt}
      boxExtents=<50 60 60>
      constraintSetup {angularLimitsMax=<15 30 0>, angularLimitsMin=<-15 -10 -30>, angularTarget=<1 0 0>}
      localJointOffset=<2 0 0>
     }]
    relativeSpaceBone {boneName=pelvis}
   }
   n23 = Anim Dynamics - Bone: dyn_hairsim_R2_jnt(ComponentPose=@n22, Alpha=1, ExternalForce="0,0,0") {
    angularDampingOverride=0.9
    angularSpringConstant=500
    bAngularSpring=true
    bOverrideAngularDamping=true
    bUseGravityOverride=true
    gravityScale=0.35
    simulationSpace=Actor
    alphaBoolBlend {blendOption=Linear}
    boundBone {boneName=dyn_hairsim_R2_jnt}
    componentLinearAccScale=<1.2 1.2 1.1>
    componentLinearVelScale=<1.2 1.2 1.1>
    gravityOverride {z=-720}
    physicsBodyDefinitions [{
      sphereCollisionRadius=0
      boundBone {boneName=dyn_hairsim_R2_jnt}
      boxExtents=<50 60 60>
      constraintSetup {angularLimitsMax=<15 30 0>, angularLimitsMin=<-15 -10 -30>, angularTarget=<1 0 0>}
      localJointOffset=<2 0 0>
     }]
    relativeSpaceBone {boneName=pelvis}
   }
   Anim Dynamics - Bone: dyn_hairsim_R3_jnt(ComponentPose=@n23, Alpha=1, ExternalForce="0,0,0") {
    angularDampingOverride=0.9
    angularSpringConstant=500
    bAngularSpring=true
    bOverrideAngularDamping=true
    bUseGravityOverride=true
    gravityScale=0.35
    simulationSpace=Actor
    alphaBoolBlend {blendOption=Linear}
    boundBone {boneName=dyn_hairsim_R3_jnt}
    componentLinearAccScale=<1.2 1.2 1.1>
    componentLinearVelScale=<1.2 1.2 1.1>
    gravityOverride {z=-720}
    physicsBodyDefinitions [{
      sphereCollisionRadius=0
      boundBone {boneName=dyn_hairsim_R3_jnt}
      boxExtents=<50 60 60>
      constraintSetup {angularLimitsMax=<15 30 0>, angularLimitsMin=<-15 -10 -30>, angularTarget=<1 0 0>}
      localJointOffset=<2 0 0>
     }]
    relativeSpaceBone {boneName=pelvis}
   }
   n24 = "Constraint - dyn_hairsim_R1_jnt (P)"(ComponentPose=Local To Component(), Alpha=1, ConstraintWeights_0=1) {
    alphaBoolBlend {blendOption=Linear}
    boneToModify {boneName=dyn_hairsim_R1_jnt}
    constraintSetup [{targetBone {boneName=dyn_hairsim_1_jnt}}]
    constraintWeights [1]
   }
   n25 = "Constraint - dyn_hairsim_R2_jnt (P)"(ComponentPose=@n24, Alpha=1, ConstraintWeights_0=1) {
    alphaBoolBlend {blendOption=Linear}
    boneToModify {boneName=dyn_hairsim_R2_jnt}
    constraintSetup [{targetBone {boneName=dyn_hairsim_1_jnt}}]
    constraintWeights [1]
   }
   n26 = "Constraint - dyn_hairsim_R3_jnt (P)"(ComponentPose=@n25, Alpha=1, ConstraintWeights_0=1) {
    alphaBoolBlend {blendOption=Linear}
    boneToModify {boneName=dyn_hairsim_R3_jnt}
    constraintSetup [{targetBone {boneName=dyn_hairsim_1_jnt}}]
    constraintWeights [1]
   }
   n27 = "Constraint - dyn_hairsim_L1_jnt (P)"(ComponentPose=@n26, Alpha=1, ConstraintWeights_0=1) {
    alphaBoolBlend {blendOption=Linear}
    boneToModify {boneName=dyn_hairsim_L1_jnt}
    constraintSetup [{targetBone {boneName=dyn_hairsim_1_jnt}}]
    constraintWeights [1]
   }
   n28 = "Constraint - dyn_hairsim_L2_jnt (P)"(ComponentPose=@n27, Alpha=1, ConstraintWeights_0=1) {
    alphaBoolBlend {blendOption=Linear}
    boneToModify {boneName=dyn_hairsim_L2_jnt}
    constraintSetup [{targetBone {boneName=dyn_hairsim_1_jnt}}]
    constraintWeights [1]
   }
   n29 = "Constraint - dyn_hairsim_L3_jnt (P)"(ComponentPose=@n28, Alpha=1, ConstraintWeights_0=1) {
    alphaBoolBlend {blendOption=Linear}
    boneToModify {boneName=dyn_hairsim_L3_jnt}
    constraintSetup [{targetBone {boneName=dyn_hairsim_1_jnt}}]
    constraintWeights [1]
   }
   n30 = "Constraint - dyn_hairsim_L4_jnt (P)"(ComponentPose=@n29, Alpha=1, ConstraintWeights_0=1) {
    alphaBoolBlend {blendOption=Linear}
    boneToModify {boneName=dyn_hairsim_L4_jnt}
    constraintSetup [{targetBone {boneName=dyn_hairsim_1_jnt}}]
    constraintWeights [1]
   }
   Component To Local(ComponentPose=@n30)
   Look At - Bone: FX_hairsim_Look_at_jnt(Alpha=1) {
    alphaBoolBlend {blendOption=Linear}
    boneToModify {boneName=FX_hairsim_Look_at_jnt}
    lookAtLocation {x=0}
    lookAtTarget {boneReference {boneName=pelvis}}
    lookAt_Axis {axis=<1 0>}
   }
   n31 = "Transform (Modify) Bone - Bone: dyn_hairsim_L2_jnt"(Alpha=1, Translation="0,0,0", Rotation="25,0,0", Scale="1,1,1") {
    rotationMode=BMM_Additive
    rotationSpace=BCS_BoneSpace
    alphaBoolBlend {blendOption=Linear}
    boneToModify {boneName=dyn_hairsim_L2_jnt}
    rotation {pitch=25}
   }
   n32 = "Transform (Modify) Bone - Bone: dyn_hairsim_R1_jnt"(ComponentPose=@n31, Alpha=1, Translation="0,0,0", Rotation="50,-20,0", Scale="1,1,1") {
    rotationMode=BMM_Additive
    rotationSpace=BCS_BoneSpace
    alphaBoolBlend {blendOption=Linear}
    boneToModify {boneName=dyn_hairsim_R1_jnt}
    rotation {pitch=50, yaw=-20}
   }
   "Transform (Modify) Bone - Bone: dyn_hairsim_1_jnt"(ComponentPose=@n32, Alpha=1, Translation="0,0,0", Rotation="5,-10,0", Scale="1,1,1") {
    rotationMode=BMM_Additive
    rotationSpace=BCS_BoneSpace
    translationMode=BMM_Additive
    translationSpace=BCS_BoneSpace
    alphaBoolBlend {blendOption=Linear}
    boneToModify {boneName=dyn_hairsim_1_jnt}
    rotation {pitch=5, yaw=-10}
   }
   n33 = "Transform (Modify) Bone - Bone: dyn_hairsim_L3_jnt"(Alpha=1, Translation="0,0,0", Rotation=@n11, Scale="1,1,1") {rotationMode=BMM_Additive, alphaBoolBlend {blendOption=Linear}, alphaScaleBiasClamp {scale=0.5}, boneToModify {boneName=dyn_hairsim_L3_jnt}}
   "Transform (Modify) Bone - Bone: dyn_hairsim_L4_jnt"(ComponentPose=@n33, Alpha=1, Translation="0,0,0", Rotation=@n13, Scale="1,1,1") {rotationMode=BMM_Additive, alphaBoolBlend {blendOption=Linear}, boneToModify {boneName=dyn_hairsim_L4_jnt}}
   n34 = "Transform (Modify) Bone - Bone: dyn_hairsim_L1_jnt"(Alpha=1, Translation="0,0,0", Rotation=@n7, Scale="1,1,1") {rotationMode=BMM_Additive, alphaBoolBlend {blendOption=Linear}, alphaScaleBiasClamp {scale=0.2}, boneToModify {boneName=dyn_hairsim_L1_jnt}}
   n35 = "Transform (Modify) Bone - Bone: dyn_hairsim_L2_jnt"(ComponentPose=@n34, Alpha=1, Translation="0,0,0", Rotation=@n9, Scale="1,1,1") {rotationMode=BMM_Additive, alphaBoolBlend {blendOption=Linear}, alphaScaleBiasClamp {scale=0.2}, boneToModify {boneName=dyn_hairsim_L2_jnt}}
   "Transform (Modify) Bone - Bone: dyn_hairsim_R2_jnt"(
    ComponentPose="Transform (Modify) Bone - Bone: dyn_hairsim_R1_jnt"(ComponentPose=@n35, Alpha=1, Translation="0,0,0", Rotation=@n9, Scale="1,1,1") {rotationMode=BMM_Additive, alphaBoolBlend {blendOption=Linear}, boneToModify {boneName=dyn_hairsim_R1_jnt}}
    Alpha=1
    Translation="0,0,0"
    Rotation=@n11
    Scale="1,1,1"
   ) {rotationMode=BMM_Additive, alphaBoolBlend {blendOption=Linear}, boneToModify {boneName=dyn_hairsim_R2_jnt}}
   "Transform (Modify) Bone - Bone: dyn_hairsim_R3_jnt"(Alpha=1, Translation="0,0,0", Rotation=@n11, Scale="1,1,1") {rotationMode=BMM_Additive, alphaBoolBlend {blendOption=Linear}, boneToModify {boneName=dyn_hairsim_R3_jnt}}
   n36 = RigidBody!(
    ComponentPose=Local To Component(LocalPose=Copy Pose From Mesh() {bUseAttachedParent=true})
    Alpha
    ComponentLinearAccScale:Vector="0.3,0.3,0.3"
    ComponentLinearVelScale:Vector="0.25,0.25,0.25"
   ) -> (Pose) {
    bClampLinearTranslationLimitToRefPose=true
    bOverrideWorldGravity=true
    overridePhysicsAsset=&/F_MED_Eternal_Vanguard_FaceAcc_Physics
    simulationSpace=BaseBoneSpace
    alphaBoolBlend {blendOption=Linear}
    baseBoneRef {boneName=spine_05}
    componentAppliedLinearAccClamp=<3000 3000 3000>
    componentLinearAccScale=<0.4 0.6 0.4>
    componentLinearVelScale=<0.4 0.6 0.4>
    overrideWorldGravity {z=-270}
   }
   n37 = Control Rig(Source=Component To Local(ComponentPose=@n36)) {
    alphaBoolBlend {blendOption=Linear}
    controlRigAssetReference {blueprintRigClass=&/F_MED_Eternal_Vanguard_CtrlRig.F_MED_Eternal_Vanguard_CtrlRig_C}
   }
   Output Pose(Result=@n37) {layerGroup=None}
  ]}
 EventGraph {nodes [
   Event BlueprintUpdateAnimation()
   n1 = float * float(A=GetGameTimeInSeconds(), B=0.75) {is=title K2Node_PromotableOperator}
   n2 = float + float(A=float * float(A="Sin (Radians)"(A=@n1), B=10) {is=title K2Node_CommutativeAssociativeBinaryOperator}, B=25)
   > Set sineRotation1(sineRotation1:Rotator=MakeRotator(Pitch=@n2, Yaw=@n2)) -> (Output_Get:Rotator)
   n3 = float + float(A=@n1, B=-1)
   n4 = float + float(A=float * float(A="Sin (Radians)"(A=@n3), B=14), B=15)
   > Set sineRotation2(sineRotation2:Rotator=MakeRotator(Pitch=@n4, Yaw=@n4)) -> (Output_Get:Rotator)
   n5 = float + float(A=@n1, B=-2)
   n6 = float + float(A=float * float(A="Sin (Radians)"(A=@n5), B=20), B=10)
   > Set sineRotation3(sineRotation3:Rotator=MakeRotator(Pitch=@n6, Yaw=@n6)) -> (Output_Get:Rotator)
   n7 = float + float(A=@n1, B=-3)
   n8 = float + float(A=float * float(A="Sin (Radians)"(A=@n7), B=25), B=25)
   > Set sineRotation4(sineRotation4:Rotator=MakeRotator(Pitch=@n8, Yaw=@n8)) -> (Output_Get:Rotator)
  ]}
]
variables [
 sineRotation1 {type=Rotator}
 sineRotation2 {type=Rotator}
 sineRotation3 {type=Rotator}
 sineRotation4 {type=Rotator}
]
