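# Animation Blueprint for the masked face accessory of the female medium Eternal Vanguard cosmetic body, following the mesh it attaches to.
# It copies and caches the attached parent's pose.
# A layered bone blend pulls the neck_01 branch 70% toward the unconnected reference-pose input, then restores the copied pose at full weight from faceAttach.
# Initialization sets MeshToCopy from the owner's skeletal mesh component, but the AnimGraph never reads it.
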AnimBlueprint /BRCosmetics/Characters/Player/Female/Medium/Bodies/F_MED_Eternal_Vanguard/Meshes/Parts/F_MED_Eternal_Vanguard_Masked_FaceAcc_AnimBP
target_skeleton=&/F_MED_Eternal_Vanguard_Masked_FaceAcc_Skeleton
graphs [
 AnimGraph {nodes [
   n1 = Layered blend per bone!(
    BasePose=Use cached pose 'SavedPose7797'()
    BlendPoses_0
    BlendPoses_1:Pose Link Structure=Use cached pose 'SavedPose7797'()
    BlendWeights_0=0.7
    BlendWeights_1:float=1
   ) -> (Pose) {
    blendPoses [{linkId=-1, sourceLinkId=-1}, {linkId=-1, sourceLinkId=-1}]
    blendWeights [1,1]
    layerSetup [{branchFilters [{blendDepth=1, boneName=neck_01}]}, {branchFilters [{blendDepth=1, boneName=faceAttach}]}]
   }
   Output Pose(Result=@n1) {layerGroup=None}
   Save cached pose 'SavedPose7797'(Pose=Copy Pose From Mesh() {bUseAttachedParent=true}) {cacheName=SavedPose7797}
  ]}
 EventGraph {nodes [
   Event BlueprintInitializeAnimation()
   > Set MeshToCopy(MeshToCopy:Skeletal Mesh Component Object Reference=GetOwnerSkeletalMeshComponent()) -> (Output_Get:Skeletal Mesh Component Object Reference)
   Event BlueprintUpdateAnimation()
  ]}
]
variables [MeshToCopy {type=SkeletalMeshComponent}]
